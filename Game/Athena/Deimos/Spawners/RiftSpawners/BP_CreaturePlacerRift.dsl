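; BP_CreaturePlacerRift is a rift spawner variant, apparently for placed creature spawns, that hides and disables its physical presence.
; On BeginPlay it runs parent logic, then HideSpawner hides the cube and fog, disables mesh and root sphere collision, and sets bHideMesh.
; Cosmetic state and encounter spawn overrides call parent effects only while bHidesVFX is false, which HideVFX sets true.
; The four cosmetic state overrides also contain debug prints.
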(event EventBeginPlay
  ;[1 1 1] EVENT -> BeginPlay
  (|CallParentFunction)
  (CallFunction|HideSpawner))

(event EventBP_Cosmetic_Intro
  ;[0.5937 1 0.5123] Replicated Cosmetic Events
  ; Calling SetCosmeticState replicates to these.
  (if (not (Variables|Default|GetbHidesVFX))
    (|CallParentFunction)
    (Development|PrintString "BP_CreaturePlacerRift: Cosmetic Intro" false false)))

(event EventBP_Cosmetic_Idle
  ;[0.5937 1 0.5123] Replicated Cosmetic Events
  ; Calling SetCosmeticState replicates to these.
  (if (not (Variables|Default|GetbHidesVFX))
    (|CallParentFunction)
    (Development|PrintString "BP_CreaturePlacerRift: Cosmetic Idle" false false)))

(event EventBP_Cosmetic_RampUp
  ;[0.5937 1 0.5123] Replicated Cosmetic Events
  ; Calling SetCosmeticState replicates to these.
  (if (not (Variables|Default|GetbHidesVFX))
    (|CallParentFunction)
    (Development|PrintString "BP_CreaturePlacerRift: Cosmetic Ramp Up" false false)))

(event EventBP_Cosmetic_ShouldDie
  ;[0.5937 1 0.5123] Replicated Cosmetic Events
  ; Calling SetCosmeticState replicates to these.
  (if (not (Variables|Default|GetbHidesVFX))
    (|CallParentFunction)
    (Development|PrintString "BP_CreaturePlacerRift: Cosmetic Should Die" false false)))

(event BuildingRift|EventOnStartedEncounterSpawn
  (params (SpawnedAI :type Fort%20AIPawn%20Object%20Reference))
  ;[0.5937 1 0.5123] Replicated Cosmetic Events
  ; Calling SetCosmeticState replicates to these.
  (if (not (Variables|Default|GetbHidesVFX))
    (|CallParentFunction SpawnedAI)))

(event EventBlueprintHideVFX
  (Variables|Default|SetbHidesVFX true))

(event Custom|HideSpawner
  (Rendering|SetVisibility (Variables|Default|GetCubeMesh) false true)
  (Collision|SetCollisionEnabled (Variables|Default|GetCubeMesh) "NoCollision")
  (Collision|SetCollisionEnabled (Variables|Default|GetCosmeticBaseMesh) "NoCollision")
  (Rendering|SetVisibility (Variables|Default|GetP_Rift_IdleFog) false false)
  (Rendering|SetVisibility (Variables|Default|GetFogMesh) false false)
  (bind _assphere_collision (Utilities|Casting|CastToSphereComponent (Variables|Transformation|GetRootComponent)))
  (Collision|SetCollisionEnabled _assphere_collision "NoCollision")
  (Variables|Default|SetbHideMesh true))

(fn ConstructionScript)
