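; Creative-mode Deimos cube rift AI spawner that can be shown, hidden and given an optional lifetime.
; Replicated bShowRift, used so replay scrubbing works, drives mesh and fog visibility, spawn effects, and collision restored from defaults cached in the construction script.
; The server runs the lifetime timer, restarts it on throttled damage, and on expiry applies a GameplayEffect and destroys the actor.
; A bHidesVFX flag suppresses inherited cosmetic state events.
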(event EventBeginPlay
  ;[1 1 1] EVENT -> BeginPlay
  (|CallParentFunction)
  (Utilities|FlowControl|Sequence
    (:then_0
      ;;[0.125 0.5 0 0.5] Want to assume spawner is invisible by default, then set to visible later
      (CallFunction|SetSpawnerVisibility (Variables|Transient|GetbShowRift)))
    (:then_1
      (Networking|SwitchHasAuthority
        (:Authority
          (CallFunction|SetLifeTimeTimer))))))

(event Custom|SetLifeTimeTimer
  ;[1 1 1] CUSTOM EVENT -> SetLifeTimeTimer
  (if (Variables|Default|GetbHasLifeTime)
    (Utilities|Time|ClearandInvalidateTimerbyHandle (Variables|Default|GetLifeTimeTimerHandle))
    (bind _returnvalue (Utilities|Time|SetTimerbyEvent (EventDispatchers|CreateEvent :!event Custom|LifeTimeExpired) (Variables|Default|GetLifeTime)))
    (Variables|Default|SetLifeTimeTimerHandle _returnvalue)))

(event EventOnDamageServer
  (params (Damage :type Float%20%28single-precision%29) (DamageTags :type Gameplay%20Tag%20Container%20Structure%20%28by%20ref%29) (Momentum :type Vector :default "0, 0, 0") (HitInfo :type Hit%20Result%20Structure%20%28by%20ref%29) (InstigatedBy :type Controller%20Object%20Reference) (DamageCauser :type Actor%20Object%20Reference) (EffectContext :type Gameplay%20Effect%20Context%20Handle%20Structure))
  ;[1 1 1] EVENT -> OnDamageServer
  (if (> (Utilities|TimeSince (Variables|Default|GetTimeLastSetLifeTimeTimer)) (Variables|Default|GetThrottleTime))
    (Variables|Default|SetTimeLastSetLifeTimeTimer (Utilities|Time|GetTimeSeconds))
    (CallFunction|SetLifeTimeTimer)))

(event Custom|LifeTimeExpired
  ;[1 1 1] CUSTOM EVENT -> LifeTimeExpired
  (GameplayEffects|ApplyGameplayEffectToSelf :self (Ability|GetAbilitySystemComponent self) :GameplayEffectClass (Variables|GameplayEffects|GetGE_LifeTimeLapsed) :Level 0.0)
  (Actor|DestroyActor))

(event EventBP_Cosmetic_Intro
  ;[0.5937 1 0.5123] Replicated Cosmetic Events
  ; Calling SetCosmeticState replicates to these.
  (if (not (Variables|Default|GetbHidesVFX))
    (|CallParentFunction)
    (Development|PrintString "BP_DeimosRift_Dynamic: Cosmetic Intro" false false)))

(event EventBP_Cosmetic_Idle
  ;[0.5937 1 0.5123] Replicated Cosmetic Events
  ; Calling SetCosmeticState replicates to these.
  (if (not (Variables|Default|GetbHidesVFX))
    (|CallParentFunction)
    (Development|PrintString "BP_DeimosRift_Dynamic: Cosmetic Idle" false false)))

(event EventBP_Cosmetic_RampUp
  ;[0.5937 1 0.5123] Replicated Cosmetic Events
  ; Calling SetCosmeticState replicates to these.
  (if (not (Variables|Default|GetbHidesVFX))
    (|CallParentFunction)
    (Development|PrintString "BP_DeimosRift_Dynamic: Cosmetic Ramp Up" false false)))

(event EventBP_Cosmetic_ShouldDie
  ;[0.5937 1 0.5123] Replicated Cosmetic Events
  ; Calling SetCosmeticState replicates to these.
  (if (not (Variables|Default|GetbHidesVFX))
    (|CallParentFunction)
    (Development|PrintString "BP_DeimosRift_Dynamic: Cosmetic Should Die" false false)))

(event BuildingRift|EventOnStartedEncounterSpawn
  (params (SpawnedAI :type Fort%20AIPawn%20Object%20Reference))
  ;[0.5937 1 0.5123] Replicated Cosmetic Events
  ; Calling SetCosmeticState replicates to these.
  (if (not (Variables|Default|GetbHidesVFX))
    (|CallParentFunction SpawnedAI)))

(event EventBlueprintHideRift
  ;[1 0.125 0 0.5] Replays won't receive these events during scrubbing, so using a replicated variable
  (Variables|Transient|SetbShowRift false))

(event EventBlueprintShowRift
  ;[1 0.125 0 0.5] Replays won't receive these events during scrubbing, so using a replicated variable
  (Variables|Transient|SetbShowRift true)
  (CallFunction|TrySpawnEffects))

(event EventBlueprintHideVFX
  (Variables|Default|SetbHidesVFX true))

(fn OnRep_bShowRift
  (CallFunction|SetSpawnerVisibility (Variables|Transient|GetbShowRift)))

(fn OnRep_ServerSpawnTime)

(fn SetCollisionForSubComponent
  (params (Component :type Primitive%20Component%20Object%20Reference) (bIsVisible :type Boolean))
  (Utilities|IsValid Component
    (:Is%20Valid
      (Collision|SetCollisionEnabled (Variables|GetComponent) (select (Variables|GetbIsVisible) (Utilities|Map|Find (Variables|Default|GetCollsionDefaults) (Variables|GetComponent)) "NoCollision")))))

(fn SetSpawnerVisibility
  (params (bIsVisible :type Boolean))
  (Utilities|FlowControl|Sequence
    (:then_0
      (Rendering|SetVisibility (Variables|Default|GetCubeMesh) (Variables|GetbIsVisible) true)
      (Rendering|SetVisibility (Variables|Default|GetP_Rift_IdleFog) (Variables|GetbIsVisible) false))
    (:then_1
      (Variables|Default|SetbHideMesh (not (Variables|GetbIsVisible))))
    (:then_2
      (Utilities|FlowControl|Sequence
        (:then_0
          (CallFunction|SetCollisionforSubComponent (Variables|Default|GetCosmeticBaseMesh) (Variables|GetbIsVisible))
          (CallFunction|SetCollisionforSubComponent (Variables|Default|GetCubeMesh) (Variables|GetbIsVisible)))
        (:then_2
          (bind _asprimitive_component (Utilities|Casting|CastToPrimitiveComponent (Variables|Transformation|GetRootComponent)))
          (CallFunction|SetCollisionforSubComponent _asprimitive_component (Variables|GetbIsVisible)))))))

(fn ShouldPlaySpawnEffects
  (return (and (|CallParentFunction) (Variables|Transient|GetbShowRift))))

(fn ConstructionScript
  (bind _p_deimos_riftspawn_damagestate (Variables|Default|GetP_Deimos_RiftSpawn_DamageState))
  (|CallParentFunction)
  (Utilities|FlowControl|Sequence
    (:then_0
      (Effects|Components|ParticleSystem|SetVectorParameter _p_deimos_riftspawn_damagestate "SizeScale" "0.700000,0.700000,0.700000")
      (Effects|Components|ParticleSystem|SetFloatParameter _p_deimos_riftspawn_damagestate "Radius" 0.0)
      (Components|DestroyComponent (Variables|Default|GetSM_CubeRunes)))
    (:then_1
      (Variables|Default|SetCollsionDefaults (Utilities|Map|MakeMap (Variables|Default|GetCubeMesh) (Collision|GetCollisionEnabled (Variables|Default|GetCubeMesh)) (Variables|Default|GetCosmeticBaseMesh) (Collision|GetCollisionEnabled (Variables|Default|GetCosmeticBaseMesh))))
      (bind _asprimitive_component (Utilities|Casting|CastToPrimitiveComponent (Variables|Transformation|GetRootComponent)))
      (Utilities|Map|Add (Variables|Default|GetCollsionDefaults) _asprimitive_component (Collision|GetCollisionEnabled _asprimitive_component)))))
